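# Translucent, two-sided unlit material for wing-like air trail effects on Niagara mesh particles and ribbons in the GarlicWhisk emote.
# Panning smoke noise distorts scrolling plasma-coil and smoke-trail textures, tinted for emissive; a rainbow noise branch appears unused.
# Opacity combines smoke, sine edge shaping, particle alpha, depth and view-angle fades, noise breakup and a MaskProgress dynamic-parameter wipe.
# World position offset adds Z-axis rotation and a U-weighted travelling sine wave.
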Material /BRCosmetics/Animation/Game/MainPlayer/Emotes/GarlicWhisk/FX/MasamuneMats/M_Masamune_AirSprite_Wing
graph {
 properties {
  bUsedWithNiagaraMeshParticles=true
  bUsedWithNiagaraRibbons=true
  blendMode=BLEND_Translucent
  pixelDepthOffsetMode=PDOM_Legacy
  previewMesh=/BRCosmetics/Animation/Game/MainPlayer/Emotes/GarlicWhisk/FX/SM_Vortex_WindTrail
  shadingModel=MSM_Unlit
  twoSided=true
  usedShadingModels=Unlit
 }
 nodes [
  n1 = TextureCoordinate() {vTiling=0.5}
  n2 = Constant2Vector() {r=0.7}
  n3 = VectorParameter() {parameterName=ColorNoiseSpeed, defaultValue=<rgba -0.2 0 0 0>}
  n4 = TextureSample(UVs=Panner(Coordinate=Add(A=@n1, B=@n2), Speed=ComponentMask(None=@n3.RGB) {g=true, r=true}) {speedX=-2}) {texture=&T_Tile_RainbowColor}
  n5 = VectorParameter() {parameterName=Color, defaultValue=<rgba 1 0.9598 0.3008 1>}
  n6 = ScalarParameter() {parameterName=DefaultColorDes}
  n7 = TextureCoordinate() {uTiling=0.5, vTiling=0.5}
  n8 = TextureCoordinate()
  n9 = Multiply(A=TextureSample(UVs=Panner(Coordinate=@n8) {speedX=0.2}) {texture=&T_Inky_Smoke_Tile}.R) {constB=0.1}
  n10 = Multiply(A=Time(), B=ScalarParameter() {defaultValue=0.4, parameterName=Speed Direction}) {constB=-1.5}
  n11 = TextureSample(UVs=Panner(Coordinate=Add(A=@n9, B=@n7), Time=@n10) {speedX=-2}) {samplerType=SAMPLERTYPE_LinearGrayscale, texture=/Game/Effects/Fort_Effects/Emotes/UltraLight/Texture/T_PlasmaCoils}
  n12 = TextureCoordinate()
  n13 = TextureSample(UVs=Panner(Coordinate=Add(A=@n12, B=@n9), Time=@n10) {speedX=-2}) {samplerType=SAMPLERTYPE_Grayscale, texture=&T_Smoke_Trail_Soft}
  n14 = Add(
   A=Multiply(A=Desaturation(None=@n5.RGB, Fraction=@n6), B=Power(Base=@n11.R))
   B=Multiply(A=VectorParameter() {parameterName=ColorDark, defaultValue=<rgba 0.401 0.401 0.401 1>}.RGB, B=@n13.R)
  )
  Add(A=Desaturation(None=Power(Base=@n4.RGB), Fraction=Constant()), B=@n14)
  MaterialFunctionCall(Emissive=Multiply(B=@n14) {constA=1}) {materialFunction=&MF_FX_EmissiveAdjustments}
  n15 = Add(A=@n9, B=TextureCoordinate())
  n16 = Multiply(
   A=TextureSample(UVs=Panner(Coordinate=@n15, Time=@n10) {speedX=-2.5}) {samplerType=SAMPLERTYPE_Grayscale, texture=&LowResBlurredNoise}.R
   B=@n13.R
  )
  n17 = ComponentMask(None=TextureCoordinate()) {r=true}
  n18 = Multiply(A=ParticleColor().A, B=Multiply(A=Add(A=@n13.R, B=@n16) {constB=0}, B=Sine(None=@n17) {period=2}))
  n19 = DepthFade(Opacity=@n18, FadeDistance=ScalarParameter() {defaultValue=10, parameterName=FadeDistance}) {fadeDistanceDefault=10}
  n20 = Multiply(
   A=ShadingPathSwitch(Default=@n19, Deferred=@n19, Forward=@n18, Mobile=@n18)
   B=MaterialFunctionCall("FadeSensitivity (optional)"=ScalarParameter() {defaultValue=1, parameterName="FadeSensitivity (optional) (S)"}) {materialFunction=&FadeBasedOnViewAngle}.Result
  )
  n21 = ScalarParameter() {defaultValue=0.7, parameterName=AlphaIntensity}
  n22 = TextureCoordinate()
  n23 = Constant2Vector()
  n24 = LinearInterpolate(Alpha=TextureSample(UVs=Add(A=@n22, B=@n23)) {samplerType=SAMPLERTYPE_Grayscale, texture=&LowResBlurredNoise}.R) {constA=-0.1, constB=2}
  n25 = ComponentMask(None=TextureCoordinate()) {r=true}
  n26 = DynamicParameter() {paramNames [MaskProgress,Param2,Param3,Param4]}
  Multiply(A=Multiply(A=Multiply(A=@n20, B=@n21) {constB=0.7}, B=Saturate(None=@n24)), B=Saturate(None=Add(A=@n25, B=@n26.MaskProgress)))
  n27 = Constant3Vector() {constant=<rgba 0 0 1 0>}
  n28 = Multiply(A=Time(), B=ScalarParameter() {defaultValue=0.2, parameterName=RotateSpeed}) {constB=0.2}
  n29 = ObjectPositionWS()
  n30 = WorldPosition()
  n31 = VectorParameter() {parameterName=WPODirection, defaultValue=<rgba 0 1 0 1>}
  n32 = TextureCoordinate()
  n33 = ScalarParameter() {parameterName=UVOffset}
  n34 = Add(A=ComponentMask(None=Add(A=@n32, B=@n33)) {r=true}, B=Multiply(A=Time()) {constB=-0.5})
  n35 = Multiply(
   A=Multiply(A=Transform(None=@n31.RGB) {transformType=TRANSFORM_View}, B=Sine(None=@n34) {period=0.5})
   B=ScalarParameter() {defaultValue=20, parameterName=WPOIntensity}
  ) {constB=10}
  n36 = ComponentMask(None=TextureCoordinate()) {r=true}
  Reroute(None=Add(A=Multiply(A=@n35, B=@n36), B=RotateAboutAxis(NormalizedRotationAxis=@n27, RotationAngle=@n28, PivotPoint=@n29, Position=@n30.XYZ)))
 ]
}
